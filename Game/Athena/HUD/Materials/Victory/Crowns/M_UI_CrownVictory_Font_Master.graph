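# Translucent UI master material for Crown Victory font, giving text a tunable three-color vertical gradient that can animate to other colors.
# Offset, Lift and Taper shape the V texture coordinate into a mask that drives a 3ColorBlend and a top-to-bottom alpha lerp.
# An Animate scalar moves color through a second 3ColorBlend toward two animated colors and lerps alpha toward AnimateAlpha, with FinalAlpha scaling the opacity result.
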Material /Game/Athena/HUD/Materials/Victory/Crowns/M_UI_CrownVictory_Font_Master
graph {
 properties {blendMode=BLEND_Translucent, materialDomain=MD_UI, pixelDepthOffsetMode=PDOM_Legacy, usedShadingModels=Default Lit}
 nodes [
  n1 = ComponentMask(None=Add(A=TextureCoordinate(), B=ScalarParameter() {parameterName=Offset})) {g=true}
  n2 = ScalarParameter() {defaultValue=1, parameterName=Lift}
  n3 = Saturate(None=Power(Base=Multiply(A=@n1, B=@n2), Exp=ScalarParameter() {defaultValue=1, parameterName=Taper}))
  MaterialFunctionCall(
   A=MaterialFunctionCall(
    A=VectorParameter() {parameterName=ColorTop, defaultValue=<rgba 0 0 1 1>}.RGB
    B=VectorParameter() {parameterName=ColorMiddle, defaultValue=<rgba 0 1 0 1>}.RGB
    C=VectorParameter() {parameterName=ColorBottom, defaultValue=<rgba 1 0 0 1>}.RGB
    Alpha=@n3
   ) {materialFunction=&3ColorBlend}.Result
   B=VectorParameter() {parameterName=ColorAnimatedMid, defaultValue=<rgba 1 1 1 1>}.RGB
   C=VectorParameter() {parameterName=ColorAnimated, defaultValue=<rgba 1 1 1 1>}.RGB
   Alpha=ScalarParameter() {parameterName=Animate}
  ) {materialFunction=&3ColorBlend}
  n4 = LinearInterpolate(A=ScalarParameter() {defaultValue=1, parameterName=TopAlpha}, B=ScalarParameter() {defaultValue=1, parameterName=BottomAlpha}, Alpha=@n3)
  n5 = ScalarParameter() {defaultValue=1, parameterName=AnimateAlpha}
  n6 = ScalarParameter() {parameterName=Animate}
  Saturate(None=Multiply(A=LinearInterpolate(A=@n4, B=@n5, Alpha=@n6), B=ScalarParameter() {defaultValue=1, parameterName=FinalAlpha}))
 ]
}
